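; Music encounter for Tidal Crane Lobbers that adapts 2D music to their count as listeners enter or exit its volume.
; Registering Tidal Crane Medium pawns raises the count, deaths lower it and fire a LobberDeath trigger, and component removal recounts by tag.
; Count changes through the adjust function broadcast an event that sets MetaSound Boss.Phase and applies a random stem mix from one, two or three-plus tiers with crossfade.
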(event EventInitializeMusicEncounter
  (|CallParentFunction)
  (MusicEncounter|BindEventtoOnMusicEncounterComponentRemoved (EventDispatchers|CreateEvent :!event Custom|Handle%20Component%20Removed)))

(event EventStartMusicEncounter
  (|CallParentFunction)
  (Variables|MusicEncounter|TidalCrane|GetMusic%20Audio%20Component
    (:then)
    (:else
      (bind _audio_component (MusicEncounter|SpawnMusic2D (Variables|MusicEncounter|TidalCrane|GetMusic)))
      (Variables|MusicEncounter|TidalCrane|SetMusic%20Audio%20Component _audio_component)
      (CallFunction|UpdateMetaSoundParameters (Variables|MusicEncounter|TidalCrane|GetNumber%20of%20Lobbers))
      (Default|BindEventtoOnNumberOfLobbersChanged (EventDispatchers|CreateEvent :!event Custom|Update%20MetaSound%20Parameters)))))

(event MusicEncounter|Volume|EventListenerEntered
  (|CallParentFunction)
  (CallFunction|StartMusicEncounter))

(event MusicEncounter|Volume|EventListenerExited
  (|CallParentFunction)
  (CallFunction|PauseMusicEncounter))

(event Custom|Update%20MetaSound%20Parameters
  (params (Number%20of%20Lobbers :type Integer))
  ;[1 1 1] If no lobbers, play idle music. Else, play combat music and set a stem mix based on the count
  (MusicEncounter|TidalCrane|SetStemMix (MusicEncounter|TidalCrane|SelectNewStemMix))
  (bind _music_audio_component (Variables|MusicEncounter|TidalCrane|GetMusic%20Audio%20Component)
    (:then
      (Audio|Parameter|SetIntegerParameter _music_audio_component "Boss.Phase" (Math|Integer|SelectInt 1 0 (> (Variables|MusicEncounter|TidalCrane|GetNumber%20of%20Lobbers) 0)) :!class /Script/AudioExtensions.AudioParameterControllerInterface))
    (:else)))

(event Custom|On%20Lobber%20Died
  (params (DeadActor :type Actor%20Object%20Reference) (Damage :type Float%20%28single-precision%29) (InstigatedBy :type Fort%20Pawn%20Object%20Reference) (HitLocation :type Vector) (DamageCauser :type Actor%20Object%20Reference))
  ;[1 1 1] Cache dead lobber components so we can prevent the count from being decremented twice.
  (Utilities|IsValid (Actor|GetComponentbyClass DeadActor "/MusicEncounter/Components/BP_MusicEncounterComponent.BP_MusicEncounterComponent_C")
    (:Is%20Valid
      (MusicEncounter|TidalCrane|AdjustNumberOfLobbers -1)
      (bind _music_audio_component (Variables|MusicEncounter|TidalCrane|GetMusic%20Audio%20Component)
        (:then
          (Audio|Parameter|ExecuteTriggerParameter _music_audio_component "LobberDeath"))
        (:else)))))

(event Custom|Handle%20Component%20Removed
  (params (Music%20Encounter%20Component :type Music%20Encounter%20Component%20Base%20Object%20Reference))
  ;[1 1 1] Count the number of lobber components that remain
  (Variables|MusicEncounter|TidalCrane|SetNumber%20of%20Lobbers (MusicEncounter|TidalCrane|GetNumberOfLobberComponents)))

(fn AddMusicEncounterComponent
  (params (Music%20Encounter%20Component :type Music%20Encounter%20Component%20Base%20Object%20Reference%20%28by%20ref%29))
  (bind _aspawn_tidal_crane_medium (Utilities|Casting|CastToPawn_TidalCrane_Medium (Components|GetOwner Music%20Encounter%20Component))
    (:then
      (EventDispatchers|BindEventtoOnDeathEffects _aspawn_tidal_crane_medium (EventDispatchers|CreateEvent :!event On%20Lobber%20Died))
      (MusicEncounter|TidalCrane|AdjustNumberOfLobbers 1)
      (|CallParentFunction (Variables|GetMusic%20Encounter%20Component))
      (return))
    (:CastFailed
      (|CallParentFunction (Variables|GetMusic%20Encounter%20Component))
      (return))))

(fn AdjustNumberofLobbers
  (params (Delta :type Integer))
  (bind _output_get (LocalVariables|SetNew%20Number%20of%20Lobbers (+ (Variables|GetDelta) (Variables|MusicEncounter|TidalCrane|GetNumber%20of%20Lobbers))))
  (if (!= _output_get (Variables|MusicEncounter|TidalCrane|GetNumber%20of%20Lobbers))
    (bind _output_get_2 (Variables|MusicEncounter|TidalCrane|SetNumber%20of%20Lobbers (Math|Integer|Max(Integer) (LocalVariables|GetNew%20Number%20of%20Lobbers))))
    (Default|CallOnNumberOfLobbersChanged _output_get_2)))

(fn GetNumberofLobberComponents
  (for _array_element (MusicEncounter|GetMusicEncounterComponents)
    (if (== (Variables|/MusicEncounter/Components/BP_MusicEncounterComponent.BP_MusicEncounterComponent_C|GetComponent%20Tag _array_element) (Variables|MusicEncounter|TidalCrane|GetLobber%20Component%20Tag))
      (Math|Integer|IncrementInt (LocalVariables|GetNum%20Lobber%20Components))))
  (return (LocalVariables|GetNum%20Lobber%20Components)))

(fn GetNumberofLobbers
  (return (Variables|MusicEncounter|TidalCrane|GetNumber%20of%20Lobbers)))

(fn SelectNewStemMix
  (return (Utilities|Array|RandomArrayItem (select (Math|Integer|InRange(Integer) (Math|Integer|Absolute(Integer) (Variables|MusicEncounter|TidalCrane|GetNumber%20of%20Lobbers)) 1 3) (Utilities|Select :Index (Math|Integer|Absolute(Integer) (Variables|MusicEncounter|TidalCrane|GetNumber%20of%20Lobbers)) :Option%200 (Variables|MusicEncounter|TidalCrane|GetStem%20Mixes%20%281%20Lobber%29) :Option%201 (Variables|MusicEncounter|TidalCrane|GetStem%20Mixes%20%281%20Lobber%29) :Option%202 (Variables|MusicEncounter|TidalCrane|GetStem%20Mixes%20%282%20Lobbers%29) :Option%203 (Variables|MusicEncounter|TidalCrane|GetStem%20Mixes%20%283+%20Lobbers%29)) (Variables|MusicEncounter|TidalCrane|GetStem%20Mixes%20%283+%20Lobbers%29)))))

(fn SetStemMix
  (params (New%20Stem%20Mix :type Music%20Encounter%20Stem%20Mix%20Structure))
  (Utilities|FlowControl|Sequence
    (:then_0
      ;[1 1 1] Set crossfade time
      (Audio|Parameter|SetFloatParameter (LocalVariables|GetAudio%20Component) "Mix.CrossfadeTime" (Utilities|Struct|BreakMusicEncounterStemMix (Variables|GetNew%20Stem%20Mix)) :!class /Script/AudioExtensions.AudioParameterControllerInterface))
    (:then_1
      (Utilities|FlowControl|Sequence
        (:then_0
          ;[1 1 1] Set stem volumes
          (for _array_element (Utilities|Struct|BreakMusicEncounterStemMix (Variables|GetNew%20Stem%20Mix))
            (Audio|Parameter|SetFloatParameter (LocalVariables|GetAudio%20Component) (Utilities|Struct|BreakAudioParameter _array_element) (.FloatParam (Utilities|Struct|BreakAudioParameter _array_element)) :!class /Script/AudioExtensions.AudioParameterControllerInterface)))))))
