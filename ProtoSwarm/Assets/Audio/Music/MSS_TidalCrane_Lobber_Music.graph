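# MetaSound music source for the Lobber boss encounter, switching stem sets with Boss.Phase and playing a death stinger.
# Harmonix metronomes drive looping Phase 0 and 1 stems, a beat-quantized transition stinger that launches Phase 1, and the scheduled death stinger.
# Phase 1 stems start and stop with their interpolated dynamic gains, seeking to the clock position.
# Weapon-fire modulation ducks the output; pausing fades it out, then finishes the source.
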MetaSoundSource /ProtoSwarm/Assets/Audio/Music/MSS_TidalCrane_Lobber_Music
inputs [
 Boss.Phase {type=Int32}
 DuckingStrength {type=Float}
 FadeInTime {type=Time}
 FadeOutTime {type=Time}
 LobberDeath {type=Trigger}
 MasterVolume {type=Float}
 Mix.CrossfadeTime {type=Time}
 Mix.Ending {type=Float}
 Mix.Phase 0.Dynamic.Else {type=Float}
 Mix.Phase 0.Dynamic.Perc {type=Float}
 Mix.Phase 1.Dynamic.Else {type=Float}
 Mix.Phase 1.Dynamic.Extra {type=Float}
 Mix.Phase 1.Dynamic.ExtraPerc {type=Float}
 Mix.Phase 1.Dynamic.Perc {type=Float}
 Mix.Phase0.Else {type=Float}
 Mix.Phase0.Master {type=Float}
 Mix.Phase0.Perc {type=Float}
 Mix.Phase1.Else {type=Float}
 Mix.Phase1.Extra {type=Float}
 Mix.Phase1.ExtraPerc {type=Float}
 Mix.Phase1.Master {type=Float}
 Mix.Phase1.Perc {type=Float}
 Mix.Transition0to1 {type=Float}
 Music.DeathStinger {type=WaveAsset}
 Music.Phase0.Else {type=WaveAsset}
 Music.Phase0.Perc {type=WaveAsset}
 Music.Phase1.Else {type=WaveAsset}
 Music.Phase1.Extra {type=WaveAsset}
 Music.Phase1.ExtraPerc {type=WaveAsset}
 Music.Phase1.Perc {type=WaveAsset}
 Music.Transition0to1 {type=WaveAsset}
 MusicEncounter.bIsPaused {type=Bool}
 UE.Source.OnPlay {type=Trigger}
]
interfaces [UE.OutputFormat.Stereo:1.0, UE.Source.OneShot:1.0, UE.Source:1.1]
outputs [
 "UE.OutputFormat.Stereo.Audio:0" {type=Audio}
 "UE.OutputFormat.Stereo.Audio:1" {type=Audio}
 UE.Source.OneShot.OnFinished {type=Trigger}
]
pages [Default {
  input_defaults {
   Boss.Phase=0
   DuckingStrength=1.9764
   FadeInTime=2
   FadeOutTime=3
   LobberDeath=null
   MasterVolume=0.7606
   Mix.CrossfadeTime=1
   Mix.Ending=0.8484
   Mix.Phase 0.Dynamic.Else=1
   Mix.Phase 0.Dynamic.Perc=1
   Mix.Phase 1.Dynamic.Else=0
   Mix.Phase 1.Dynamic.Extra=0
   Mix.Phase 1.Dynamic.ExtraPerc=0
   Mix.Phase 1.Dynamic.Perc=0
   Mix.Phase0.Else=1
   Mix.Phase0.Master=1.0399
   Mix.Phase0.Perc=1
   Mix.Phase1.Else=1
   Mix.Phase1.Extra=1
   Mix.Phase1.ExtraPerc=1
   Mix.Phase1.Master=0.8883
   Mix.Phase1.Perc=1
   Mix.Transition0to1=0.984
   Music.DeathStinger=/ProtoSwarm/Assets/Audio/SoundWaves/Music/MX_S37_Combat_ArmoredBug_Death_STNG
   Music.Phase0.Else=/ProtoSwarm/Assets/Audio/SoundWaves/Music/MX_S37_Combat_ArmoredBug_Phase_0_LP_Else
   Music.Phase0.Perc=/ProtoSwarm/Assets/Audio/SoundWaves/Music/MX_S37_Combat_ArmoredBug_Phase_0_LP_Perc
   Music.Phase1.Else=/ProtoSwarm/Assets/Audio/SoundWaves/Music/MX_S37_Combat_ArmoredBug_Phase_1_LP_Else
   Music.Phase1.Extra=/ProtoSwarm/Assets/Audio/SoundWaves/Music/MX_S37_Combat_ArmoredBug_Phase_1_LP_Extra
   Music.Phase1.ExtraPerc=/ProtoSwarm/Assets/Audio/SoundWaves/Music/MX_S37_Combat_ArmoredBug_Phase_1_LP_Perc_2
   Music.Phase1.Perc=/ProtoSwarm/Assets/Audio/SoundWaves/Music/MX_S37_Combat_ArmoredBug_Phase_1_LP_Perc
   Music.Transition0to1=/ProtoSwarm/Assets/Audio/SoundWaves/Music/MX_S37_Combat_ArmoredBug_Phase_0_to_1_STNG
   MusicEncounter.bIsPaused=null
   UE.Source.OnPlay=null
  }
  nodes [
   InitVariable.MusicSeekTarget()
   InitVariable.Trigger()
   InitVariable.Trigger()
   InitVariable.Trigger()
   InitVariable.Trigger()
   InitVariable.Trigger()
   InitVariable.Trigger()
   InitVariable.Trigger()
   TriggerOnValueChange_6 = UE.Trigger On Value Change:Int32(Value=^Boss.Phase)
   TriggerAnyTriggerAny2_1 = "TriggerAny.Trigger Any (2)"(In 1=^UE.Source.OnPlay, In 0=@TriggerOnValueChange_6.Trigger)
   n1 = VariableAccessor.Trigger(Variable=VariableMutator.Trigger(Variable=InitVariable.Trigger().Variable, Value=TriggerCompare.Clamp:Int32(A=^Boss.Phase, Compare=@TriggerAnyTriggerAny2_1.Out).True).Variable)
   n2 = VariableAccessor.Trigger(Variable=@n1.Variable)
   n3 = VariableAccessor.Trigger(Variable=@n2.Variable)
   n4 = VariableDeferredAccessor.Trigger(Variable=InitVariable.Trigger().Variable)
   n5 = VariableDeferredAccessor.Trigger(Variable=InitVariable.Trigger().Variable)
   TriggerAnyTriggerAny2_2 = "TriggerAny.Trigger Any (2)"(In 0=@n4.Value, In 1=@n5.Value)
   n6 = VariableDeferredAccessor.Trigger(Variable=InitVariable.Trigger().Variable)
   TriggerControl_12 = UE.Trigger Control(Trigger In=@TriggerAnyTriggerAny2_2.Out, Open=VariableAccessor.Trigger(Variable=@n3.Variable).Value, Close=@n6.Value)
   TriggerDelay_2 = UE.Trigger Delay(In=@TriggerControl_12.Trigger Out) {also {defaults {Delay Time=0.3}}}
   n7 = HarmonixNodes.TriggerToTransport(
    Play=@n3.Value
    Stop=VariableAccessor.Trigger(Variable=VariableMutator.Trigger(Value=@TriggerDelay_2.Out, Variable=@n6.Variable).Variable).Value
   )
   n8 = HarmonixNodes.Metronome(Transport=@n7.Transport) {also {defaults {"Loop Length (Bars)"=18, Simple Loop="<MetaSoundBuilderResult.SUCCEEDED: 0>", Tempo=146}}}
   n9 = VariableAccessor.MIDIClock(Variable=VariableMutator.MIDIClock(Value=@n8.MIDI Clock, Variable=InitVariable.MIDIClock().Variable).Variable)
   n10 = VariableAccessor.MIDIClock(Variable=@n9.Variable)
   n11 = VariableAccessor.MusicTransport(Variable=VariableMutator.MusicTransport(Value=@n7.Transport, Variable=InitVariable.MusicTransport().Variable).Variable)
   n12 = HarmonixNodes.TransportWavePlayerController(MIDI Clock=@n10.Value, Transport=VariableAccessor.MusicTransport(Variable=@n11.Variable).Value) {also {version [0,1]}}
   WavePlayer_0 = UE.Wave Player:Stereo(Wave Asset=^Music.Phase0.Perc, Play=@n12.Play, Start Time=@n12.Start Time, Stop=@n12.Stop) {also {defaults {Loop="<MetaSoundBuilderResult.SUCCEEDED: 0>"}}}
   n13 = HarmonixNodes.TransportWavePlayerController(MIDI Clock=@n9.Value, Transport=@n11.Value) {also {version [0,1]}}
   WavePlayer_3 = UE.Wave Player:Stereo(Wave Asset=^Music.Phase0.Else, Play=@n13.Play, Start Time=@n13.Start Time, Stop=@n13.Stop) {also {defaults {Loop="<MetaSoundBuilderResult.SUCCEEDED: 0>"}}}
   AudioMixerAudioMixerStereo2_0 = "AudioMixer.Audio Mixer (Stereo, 2)"(
    Gain 0=UE.Multiply:Float(AdditionalOperands=^Mix.Phase 0.Dynamic.Else, PrimaryOperand=^Mix.Phase0.Else).Out
    Gain 1=UE.Multiply:Float(AdditionalOperands=^Mix.Phase 0.Dynamic.Perc, PrimaryOperand=^Mix.Phase0.Perc).Out
    In 1 L=@WavePlayer_0.Out Left
    In 1 R=@WavePlayer_0.Out Right
    In 0 L=@WavePlayer_3.Out Left
    In 0 R=@WavePlayer_3.Out Right
   )
   n14 = UE.InterpTo:Audio(Interp Time=^Mix.CrossfadeTime, Target=^Mix.Phase 1.Dynamic.ExtraPerc)
   n15 = VariableAccessor.Float(Variable=VariableMutator.Float(Variable=InitVariable.Float().Variable, Value=@n14.Value).Variable)
   n16 = UE.InterpTo:Audio(Interp Time=^Mix.CrossfadeTime, Target=^Mix.Phase 1.Dynamic.Extra)
   n17 = VariableAccessor.Float(Variable=VariableMutator.Float(Variable=InitVariable.Float().Variable, Value=@n16.Value).Variable)
   n18 = UE.InterpTo:Audio(Interp Time=^Mix.CrossfadeTime, Target=^Mix.Phase 1.Dynamic.Else)
   n19 = VariableAccessor.Float(Variable=VariableMutator.Float(Variable=InitVariable.Float().Variable, Value=@n18.Value).Variable)
   n20 = UE.InterpTo:Audio(Interp Time=^Mix.CrossfadeTime, Target=^Mix.Phase 1.Dynamic.Perc)
   n21 = VariableAccessor.Float(Variable=VariableMutator.Float(Variable=InitVariable.Float().Variable, Value=@n20.Value).Variable)
   TriggerOnValueChange_5 = UE.Trigger On Value Change:Int32(Value=^Boss.Phase)
   n22 = VariableDeferredAccessor.Trigger(Variable=@n5.Variable)
   TriggerControl_4 = UE.Trigger Control(
    Trigger In=HarmonixNodes.MidiClockSubdivisionTrigger(MIDI Clock=VariableAccessor.MIDIClock(Variable=@n10.Variable).Value) {also {defaults {"Base Offset Size "=4, Offset Multiplier=1}}}.Trigger Out
    Open=TriggerCompare.Clamp:Int32(A=^Boss.Phase, Compare=@TriggerOnValueChange_5.Trigger) {also {defaults {B=1}}}.True
    Close=@n22.Value
   )
   n23 = VariableDeferredAccessor.Trigger(Variable=InitVariable.Trigger().Variable)
   n24 = HarmonixNodes.TriggerToTransport(
    Play=VariableAccessor.Trigger(Variable=VariableMutator.Trigger(Value=@TriggerControl_4.Trigger Out, Variable=@n22.Variable).Variable).Value
    Stop=@n23.Value
   )
   n25 = HarmonixNodes.Metronome(Transport=@n24.Transport) {also {defaults {Tempo=146}}}
   n26 = HarmonixNodes.MidiClockTimestampTrigger(
    MIDI Clock=@n25.MIDI Clock
    Music Timestamp=HarmonixNodes.MusicTimestampBuilder() {also {defaults {Bar=2, Beat=1.5}, version [0,1]}}.Music Timestamp
   )
   n27 = InitVariable.Trigger()
   TriggerAnyTriggerAny2_7 = "TriggerAny.Trigger Any (2)"(
    In 0=TriggerCompare.Clamp:Int32(A=^Boss.Phase, Compare=^UE.Source.OnPlay) {also {defaults {B=1}}}.True
    In 1=VariableAccessor.Trigger(Variable=VariableMutator.Trigger(Value=@n26.Trigger Out, Variable=@n27.Variable).Variable).Value
   )
   n28 = VariableAccessor.Trigger(Variable=VariableMutator.Trigger(Value=@TriggerAnyTriggerAny2_7.Out, Variable=@n4.Variable).Variable)
   n29 = VariableAccessor.Trigger(Variable=@n28.Variable)
   n30 = VariableAccessor.Trigger(Variable=@n29.Variable)
   n31 = VariableAccessor.Trigger(Variable=@n30.Variable)
   n32 = VariableAccessor.Trigger(Variable=@n31.Variable)
   n33 = VariableAccessor.Trigger(Variable=@n32.Variable)
   n34 = VariableAccessor.Trigger(Variable=@n33.Variable)
   n35 = VariableDeferredAccessor.Trigger(Variable=InitVariable.Trigger().Variable)
   TriggerControl_7 = UE.Trigger Control(Trigger In=@n2.Value, Open=@n34.Value, Close=@n35.Value)
   TriggerDelay_1 = UE.Trigger Delay(In=@TriggerControl_7.Trigger Out) {also {defaults {Delay Time=3}}}
   n36 = VariableAccessor.Trigger(Variable=VariableMutator.Trigger(Value=@TriggerDelay_1.Out, Variable=@n35.Variable).Variable)
   n37 = VariableAccessor.Trigger(Variable=@n36.Variable)
   n38 = VariableAccessor.Trigger(Variable=@n37.Variable)
   n39 = VariableAccessor.Trigger(Variable=@n38.Variable)
   n40 = VariableAccessor.Trigger(Variable=@n39.Variable)
   n41 = HarmonixNodes.TriggerToTransport(Stop=@n40.Value, Play=@n33.Value)
   n42 = HarmonixNodes.Metronome(Transport=@n41.Transport) {also {defaults {"Loop Length (Bars)"=18, Simple Loop="<MetaSoundBuilderResult.SUCCEEDED: 0>", Tempo=146}}}
   n43 = HarmonixNodes.MidiClockState(MIDI Clock=@n42.MIDI Clock) {also {version [0,1]}}
   n44 = UE.Multiply:Int32(
    PrimaryOperand=@n43.Time Sig. Numerator
    AdditionalOperands=UE.Subtract:Int32(PrimaryOperand=@n43.Bar) {also {defaults {AdditionalOperands=1}}}.Out
   ) {also {version [1,1]}}
   n45 = UE.Multiply:Float(
    PrimaryOperand=UE.Add:Float(
     PrimaryOperand=Convert.Float:Int32(Int32=@n44.Out).Float
     AdditionalOperands=UE.Subtract:Float(PrimaryOperand=@n43.Beat) {also {defaults {AdditionalOperands=1}}}.Out
    ).Out
    AdditionalOperands=@n43.Secs Per Beat
   )
   n46 = VariableAccessor.Time(Variable=VariableMutator.Time(Value=UE.ConversionFloatToTime(In=@n45.Out).Out, Variable=InitVariable.Time().Variable).Variable)
   TriggerOnValueChange_4 = UE.Trigger On Value Change:Float(Value=@n18.Value)
   n47 = VariableAccessor.Trigger(Variable=@n40.Variable)
   n48 = VariableAccessor.Trigger(Variable=@n34.Variable)
   TriggerControl_11 = UE.Trigger Control(Trigger In=@TriggerOnValueChange_4.Trigger, Close=@n47.Value, Open=@n48.Value)
   TriggerAnyTriggerAny2_5 = "TriggerAny.Trigger Any (2)"(In 0=@TriggerControl_11.Trigger Out, In 1=@n29.Value)
   n49 = TriggerCompare.Clamp:Float(Compare=@TriggerAnyTriggerAny2_5.Out, B=@n19.Value) {also {defaults {Type=1}}}
   n50 = VariableDeferredAccessor.Trigger(Variable=InitVariable.Trigger().Variable)
   n51 = VariableDeferredAccessor.Trigger(Variable=InitVariable.Trigger().Variable)
   TriggerControl_10 = UE.Trigger Control(Trigger In=@n49.True, Open=@n50.Value, Close=@n51.Value)
   TriggerAnyTriggerAny2_6 = "TriggerAny.Trigger Any (2)"(In 0=@n49.False, In 1=@n38.Value)
   WavePlayer_2 = UE.Wave Player:Stereo(
    Wave Asset=^Music.Phase1.Else
    Start Time=@n46.Value
    Play=VariableAccessor.Trigger(Variable=VariableMutator.Trigger(Value=@TriggerControl_10.Trigger Out, Variable=@n51.Variable).Variable).Value
    Stop=VariableAccessor.Trigger(Variable=VariableMutator.Trigger(Value=@TriggerAnyTriggerAny2_6.Out, Variable=@n50.Variable).Variable).Value
   ) {also {defaults {Loop="<MetaSoundBuilderResult.SUCCEEDED: 0>"}}}
   n52 = VariableAccessor.Time(Variable=@n46.Variable)
   TriggerOnValueChange_1 = UE.Trigger On Value Change:Float(Value=@n20.Value)
   n53 = VariableAccessor.Trigger(Variable=@n47.Variable)
   n54 = VariableAccessor.Trigger(Variable=@n48.Variable)
   TriggerControl_0 = UE.Trigger Control(Trigger In=@TriggerOnValueChange_1.Trigger, Close=@n53.Value, Open=@n54.Value)
   TriggerAnyTriggerAny2_8 = "TriggerAny.Trigger Any (2)"(In 0=@TriggerControl_0.Trigger Out, In 1=@n30.Value)
   n55 = TriggerCompare.Clamp:Float(Compare=@TriggerAnyTriggerAny2_8.Out, B=@n21.Value) {also {defaults {A=0, Type=1}}}
   n56 = VariableDeferredAccessor.Trigger(Variable=InitVariable.Trigger().Variable)
   n57 = VariableDeferredAccessor.Trigger(Variable=InitVariable.Trigger().Variable)
   TriggerControl_3 = UE.Trigger Control(Trigger In=@n55.True, Close=@n56.Value, Open=@n57.Value)
   TriggerAnyTriggerAny2_0 = "TriggerAny.Trigger Any (2)"(In 0=@n55.False, In 1=@n39.Value)
   WavePlayer_4 = UE.Wave Player:Stereo(
    Wave Asset=^Music.Phase1.Perc
    Start Time=@n52.Value
    Play=VariableAccessor.Trigger(Variable=VariableMutator.Trigger(Value=@TriggerControl_3.Trigger Out, Variable=@n56.Variable).Variable).Value
    Stop=VariableAccessor.Trigger(Variable=VariableMutator.Trigger(Value=@TriggerAnyTriggerAny2_0.Out, Variable=@n57.Variable).Variable).Value
   ) {also {defaults {Loop="<MetaSoundBuilderResult.SUCCEEDED: 0>"}}}
   n58 = VariableAccessor.Time(Variable=@n52.Variable)
   TriggerOnValueChange_3 = UE.Trigger On Value Change:Float(Value=@n16.Value)
   n59 = VariableAccessor.Trigger(Variable=@n54.Variable)
   n60 = VariableAccessor.Trigger(Variable=@n53.Variable)
   TriggerControl_1 = UE.Trigger Control(Trigger In=@TriggerOnValueChange_3.Trigger, Open=@n59.Value, Close=@n60.Value)
   TriggerAnyTriggerAny2_10 = "TriggerAny.Trigger Any (2)"(In 0=@TriggerControl_1.Trigger Out, In 1=@n31.Value)
   n61 = TriggerCompare.Clamp:Float(Compare=@TriggerAnyTriggerAny2_10.Out, B=VariableAccessor.Float(Variable=@n17.Variable).Value) {also {defaults {A=0, Type=1}}}
   n62 = VariableDeferredAccessor.Trigger(Variable=InitVariable.Trigger().Variable)
   n63 = VariableDeferredAccessor.Trigger(Variable=InitVariable.Trigger().Variable)
   TriggerControl_5 = UE.Trigger Control(Trigger In=@n61.True, Open=@n62.Value, Close=@n63.Value)
   TriggerAnyTriggerAny2_4 = "TriggerAny.Trigger Any (2)"(In 0=@n61.False, In 1=@n37.Value)
   WavePlayer_6 = UE.Wave Player:Stereo(
    Wave Asset=^Music.Phase1.Extra
    Start Time=@n58.Value
    Play=VariableAccessor.Trigger(Variable=VariableMutator.Trigger(Value=@TriggerControl_5.Trigger Out, Variable=@n63.Variable).Variable).Value
    Stop=VariableAccessor.Trigger(Variable=VariableMutator.Trigger(Value=@TriggerAnyTriggerAny2_4.Out, Variable=@n62.Variable).Variable).Value
   ) {also {defaults {Loop="<MetaSoundBuilderResult.SUCCEEDED: 0>"}}}
   TriggerOnValueChange_2 = UE.Trigger On Value Change:Float(Value=@n14.Value)
   n64 = VariableAccessor.Trigger(Variable=@n59.Variable)
   n65 = VariableAccessor.Trigger(Variable=@n60.Variable)
   TriggerControl_6 = UE.Trigger Control(Trigger In=@TriggerOnValueChange_2.Trigger, Open=@n64.Value, Close=@n65.Value)
   TriggerAnyTriggerAny2_9 = "TriggerAny.Trigger Any (2)"(In 0=@TriggerControl_6.Trigger Out, In 1=@n32.Value)
   n66 = TriggerCompare.Clamp:Float(Compare=@TriggerAnyTriggerAny2_9.Out, B=VariableAccessor.Float(Variable=@n15.Variable).Value) {also {defaults {A=0, Type=1}}}
   n67 = VariableDeferredAccessor.Trigger(Variable=InitVariable.Trigger().Variable)
   n68 = VariableDeferredAccessor.Trigger(Variable=InitVariable.Trigger().Variable)
   TriggerControl_2 = UE.Trigger Control(Trigger In=@n66.True, Close=@n67.Value, Open=@n68.Value)
   TriggerAnyTriggerAny2_3 = "TriggerAny.Trigger Any (2)"(In 0=@n66.False, In 1=@n36.Value)
   WavePlayer_7 = UE.Wave Player:Stereo(
    Wave Asset=^Music.Phase1.ExtraPerc
    Start Time=VariableAccessor.Time(Variable=@n58.Variable).Value
    Play=VariableAccessor.Trigger(Variable=VariableMutator.Trigger(Value=@TriggerControl_2.Trigger Out, Variable=@n67.Variable).Variable).Value
    Stop=VariableAccessor.Trigger(Variable=VariableMutator.Trigger(Value=@TriggerAnyTriggerAny2_3.Out, Variable=@n68.Variable).Variable).Value
   ) {also {defaults {Loop="<MetaSoundBuilderResult.SUCCEEDED: 0>"}}}
   AudioMixerAudioMixerStereo4_0 = "AudioMixer.Audio Mixer (Stereo, 4)"(
    Gain 3=UE.Multiply:Float(PrimaryOperand=^Mix.Phase1.ExtraPerc, AdditionalOperands=@n15.Value).Out
    Gain 2=UE.Multiply:Float(PrimaryOperand=^Mix.Phase1.Extra, AdditionalOperands=@n17.Value).Out
    Gain 0=UE.Multiply:Float(PrimaryOperand=^Mix.Phase1.Else, AdditionalOperands=VariableAccessor.Float(Variable=@n19.Variable).Value).Out
    Gain 1=UE.Multiply:Float(PrimaryOperand=^Mix.Phase1.Perc, AdditionalOperands=VariableAccessor.Float(Variable=@n21.Variable).Value).Out
    In 0 L=@WavePlayer_2.Out Left
    In 0 R=@WavePlayer_2.Out Right
    In 1 L=@WavePlayer_4.Out Left
    In 1 R=@WavePlayer_4.Out Right
    In 2 L=@WavePlayer_6.Out Left
    In 2 R=@WavePlayer_6.Out Right
    In 3 L=@WavePlayer_7.Out Left
    In 3 R=@WavePlayer_7.Out Right
   )
   n69 = InitVariable.Float()
   n70 = Value.Value:Float(Reset=@n1.Value, Set=@n28.Value) {also {defaults {Init Value=0, Target Value=1}}}
   n71 = Value.Value:Float(Set=@n1.Value, Reset=@n28.Value) {also {defaults {Init Value=1, Target Value=0}}}
   TriggerRouteTriggerRouteFloat2_0 = "TriggerRoute.Trigger Route (Float, 2):Float"(
    Value 1=UE.InterpTo:Audio(Target=@n70.Output Value) {also {defaults {Interp Time=0.3}}}.Value
    Value 0=UE.InterpTo:Audio(Target=@n71.Output Value) {also {defaults {Interp Time=3}}}.Value
    Set 0=@n71.On Set
    Set 1=@n70.On Set
   )
   n72 = VariableMutator.Float(Variable=InitVariable.Float().Variable, Value=@TriggerRouteTriggerRouteFloat2_0.Value)
   n73 = InitVariable.Float()
   n74 = VariableAccessor.MIDIClock(Variable=VariableMutator.MIDIClock(Value=@n42.MIDI Clock, Variable=InitVariable.MIDIClock().Variable).Variable)
   TriggerControl_8 = UE.Trigger Control(
    Trigger In=^LobberDeath
    Close=VariableAccessor.Trigger(Variable=@n65.Variable).Value
    Open=VariableAccessor.Trigger(Variable=@n64.Variable).Value
   )
   n75 = VariableDeferredAccessor.Trigger(Variable=InitVariable.Trigger().Variable)
   n76 = VariableDeferredAccessor.Trigger(Variable=InitVariable.Trigger().Variable)
   TriggerControl_9 = UE.Trigger Control(
    Trigger In=&MSP_ScheduleTrigger(
     In Trigger=HarmonixNodes.MidiClockSubdivisionTrigger(MIDI Clock=@n74.Value) {also {defaults {Grid Size Multiplier=2, Offset Multiplier=1}}}.Trigger Out
     Schedule=@TriggerControl_8.Trigger Out
    ).Scheduled Trigger
    Open=@n75.Value
    Close=@n76.Value
   )
   WavePlayer_1 = UE.Wave Player:Stereo(
    Wave Asset=^Music.DeathStinger
    Play=VariableAccessor.Trigger(Variable=VariableMutator.Trigger(Value=@TriggerControl_9.Trigger Out, Variable=@n76.Variable).Variable).Value
   )
   n77 = InitVariable.MIDIClock()
   n78 = InitVariable.MusicTransport()
   n79 = HarmonixNodes.TransportWavePlayerController(
    MIDI Clock=VariableAccessor.MIDIClock(Variable=VariableMutator.MIDIClock(Value=@n25.MIDI Clock, Variable=@n77.Variable).Variable).Value
    Transport=VariableAccessor.MusicTransport(Variable=VariableMutator.MusicTransport(Value=@n24.Transport, Variable=@n78.Variable).Variable).Value
   ) {also {version [0,1]}}
   WavePlayer_5 = UE.Wave Player:Stereo(Wave Asset=^Music.Transition0to1, Play=@n79.Play, Start Time=@n79.Start Time, Stop=@n79.Stop)
   AudioMixerAudioMixerStereo4_1 = "AudioMixer.Audio Mixer (Stereo, 4)"(
    Gain 3=^Mix.Ending
    In 0 L=@AudioMixerAudioMixerStereo2_0.Out L
    In 0 R=@AudioMixerAudioMixerStereo2_0.Out R
    In 2 L=@AudioMixerAudioMixerStereo4_0.Out L
    In 2 R=@AudioMixerAudioMixerStereo4_0.Out R
    Gain 0=UE.Multiply:Float(PrimaryOperand=^Mix.Phase0.Master, AdditionalOperands=VariableAccessor.Float(Variable=@n69.Variable).Value).Out
    Gain 2=UE.Multiply:Float(PrimaryOperand=^Mix.Phase1.Master, AdditionalOperands=VariableAccessor.Float(Variable=@n72.Variable).Value).Out
    Gain 1=UE.Multiply:Float(PrimaryOperand=^Mix.Transition0to1, AdditionalOperands=VariableAccessor.Float(Variable=@n73.Variable).Value).Out
    In 3 L=@WavePlayer_1.Out Left
    In 3 R=@WavePlayer_1.Out Right
    In 1 L=@WavePlayer_5.Out Left
    In 1 R=@WavePlayer_5.Out Right
   )
   ADSREnvelopeADSREnvelope_0 = ADSR Envelope.ADSR Envelope:Float(Attack Time=^FadeInTime, Trigger Attack=^UE.Source.OnPlay) {also {defaults {Decay Time=0, Sustain Level=1}}}
   TriggerOnValueChange_0 = UE.Trigger On Value Change:Bool(Value=^MusicEncounter.bIsPaused)
   n80 = TriggerCompare.Clamp:Bool(A=^MusicEncounter.bIsPaused, Compare=@TriggerOnValueChange_0.Trigger) {also {defaults {B="<MetaSoundBuilderResult.SUCCEEDED: 0>"}}}
   n81 = Value.Value:Float(Reset=@n80.False, Set=@n80.True) {also {defaults {Init Value=1, Target Value=0}}}
   n82 = UE.InterpTo:Audio(Interp Time=^FadeOutTime, Target=@n81.Output Value)
   n83 = UE.Multiply:Float(
    AdditionalOperands=^DuckingStrength
    PrimaryOperand=Modulation.GetModulatorValue() {also {defaults {Modulator=/Game/Sounds/Modulation/EnvelopeFollowers/FortWeapon/MG_WeaponFireEnvFollower}}}.Out
   )
   n84 = UE.Multiply:Float(
    AdditionalOperands=UE.Multiply:Float(
     PrimaryOperand=^MasterVolume
     AdditionalOperands=UE.Multiply:Float(PrimaryOperand=@ADSREnvelopeADSREnvelope_0.Out Envelope, AdditionalOperands=@n82.Value).Out
    ).Out
    PrimaryOperand=UE.Subtract:Float(AdditionalOperands=UE.InterpTo:Audio(Target=@n83.Out) {also {defaults {Interp Time=1}}}.Value) {also {defaults {PrimaryOperand=1}}}.Out
   )
   Multiply_0 = UE.Multiply:Audio by Float(PrimaryOperand=@AudioMixerAudioMixerStereo4_1.Out L, AdditionalOperands=@n84.Out)
   Multiply_1 = UE.Multiply:Audio by Float(PrimaryOperand=@AudioMixerAudioMixerStereo4_1.Out R, AdditionalOperands=@n84.Out)
   TriggerDelay_0 = UE.Trigger Delay(Delay Time=^FadeOutTime, Reset=@n81.On Reset, In=@n81.On Set)
   n85 = HarmonixNodes.MidiClockState(MIDI Clock=@n8.MIDI Clock) {also {version [0,1]}}
   VariableMutator.MusicSeekTarget(
    Value=HarmonixNodes.BarBeatToSeekTarget(Bar=@n85.Bar, Beat=@n85.Beat) {also {version [0,1]}}.Seek Target
    Variable=InitVariable.MusicSeekTarget().Variable
   )
   VariableMutator.MusicTransport(Value=@n41.Transport, Variable=InitVariable.MusicTransport().Variable)
   VariableMutator.Trigger(Variable=@n23.Variable, Value=@WavePlayer_5.On Finished)
   VariableMutator.Trigger(Variable=@n75.Variable, Value=@WavePlayer_1.On Finished)
  ]
  links [
   {dst="@UE.OutputFormat.Stereo.Audio:0", src=Multiply_0.Out}
   {dst="@UE.OutputFormat.Stereo.Audio:1", src=Multiply_1.Out}
   {dst="@UE.Source.OneShot.OnFinished", src=TriggerDelay_0.Out}
  ]
 }]
properties {
 assetImportData=&self:AssetImportData
 numChannels=2
 outputFormat=Stereo
 sampleRate=48000
 soundClassObject=&Fort_Music_NPC
}
